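# Peely's Season 20 Week 5 shared conversation database, a short NPC dialogue exchange for the LTDstory entry tag.
# The single entry point branches to two player choices: a banana pun about a crashed truck or a reassuring line.
# Each choice is a sub-node on a terminal Speech node that shows Peely's wordless reaction, a withering look or a reassured smile.
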ConversationDatabase /NPCLibrary/NPCs/Peely/Conversations/Shared_Conversation_Peely_S20W5
entry_points [{tag=Conversation.Entry.Peely.LTDstory, destinations [#1]}]
nodes [
 ConversationEntryPointNode_7 {
  id=#1
  class=ConversationEntryPointNode
  outputs [#2, #3]
  properties {entryTag {tagName=Conversation.Entry.Peely.LTDstory}}
 }
 Speech_C_3 {
  id=#3
  class=Speech_C
  outputs []
  properties {generalConfig {defaultMessage=(Peely smiles at you, reassured.)}}
  sub_nodes [ConversationChoiceNode_5 {class=ConversationChoiceNode, properties {defaultChoiceDisplayText=It's okay. You'll figure it out.}}]
 }
 Speech_C_7 {
  id=#2
  class=Speech_C
  outputs []
  properties {generalConfig {defaultMessage=(Peely gives you a withering look.)}}
  sub_nodes [ConversationChoiceNode_8 {class=ConversationChoiceNode, properties {defaultChoiceDisplayText="You crashed that truck?! Talk about a banana split."}}]
 }
]
